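# Physics asset for the male medium Traveler cosmetic body, covering dangling upper and lower pouches on both sides of the pelvis.
# Each pouch box is constrained with narrow twist and wider swing limits to a tiny master (MSTR) anchor body.
# The anchors join a single pelvis capsule through constraints with free swing and twist.
# A left chain-base anchor also joins the pelvis, but no chain body hangs from it.
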PhysicsAsset /BRCosmetics/Characters/Player/Male/Medium/Bodies/M_MED_Traveler/Meshes/M_MED_Traveler_Physics
bodies {
 dyn_chain_base_01_l_MSTR [{shapeType=Box, extentX=0.01, extentY=0.01, extentZ=0.01}]
 dyn_pouch_lower_01_l [{shapeType=Box, extentX=17.7626, extentY=18.5478, extentZ=7.7031, center=<7.6612 -0.0002 3.4236>}]
 dyn_pouch_lower_01_l_MSTR [{shapeType=Box, extentX=0.01, extentY=0.01, extentZ=0.01}]
 dyn_pouch_lower_01_r [{shapeType=Box, extentX=17.7626, extentY=18.5478, extentZ=7.7031, center=<7.6612 -0.0002 3.4236>}]
 dyn_pouch_lower_01_r_MSTR [{shapeType=Box, extentX=0.01, extentY=0.01, extentZ=0.01}]
 dyn_pouch_upper_01_l [{shapeType=Box, extentX=9.7626, extentY=15.5478, extentZ=7.7031, center=<4.7811 -1.929 2.7618>, rotation=<rot 0 -15.0021 0>}]
 dyn_pouch_upper_01_l_MSTR [{shapeType=Box, extentX=0.01, extentY=0.01, extentZ=0.01}]
 dyn_pouch_upper_01_r [{shapeType=Box, extentX=17.7626, extentY=18.5478, extentZ=7.7031, center=<7.6612 -0.0002 3.4236>}]
 dyn_pouch_upper_01_r_MSTR [{shapeType=Box, extentX=0.01, extentY=0.01, extentZ=0.01}]
 pelvis [{shapeType=Capsule, length=15.7206, radius=11.8169, center=<-0.1307 -0.6735 -0.0227>, rotation=<rot -0.6941 -90 0.0793>}]
}
constraints [
 {
  bone1Name=dyn_pouch_upper_01_l_MSTR
  bone2Name=pelvis
  swing1LimitDegrees=45
  swing1Motion=Free
  swing2LimitDegrees=45
  swing2Motion=Free
  twistLimitDegrees=45
  twistMotion=Free
 }
 {
  bone1Name=dyn_pouch_upper_01_r_MSTR
  bone2Name=pelvis
  swing1LimitDegrees=45
  swing1Motion=Free
  swing2LimitDegrees=45
  swing2Motion=Free
  twistLimitDegrees=45
  twistMotion=Free
 }
 {
  bone1Name=dyn_chain_base_01_l_MSTR
  bone2Name=pelvis
  swing1LimitDegrees=45
  swing1Motion=Free
  swing2LimitDegrees=45
  swing2Motion=Free
  twistLimitDegrees=45
  twistMotion=Free
 }
 {
  bone1Name=dyn_pouch_lower_01_l_MSTR
  bone2Name=pelvis
  swing1LimitDegrees=45
  swing1Motion=Free
  swing2LimitDegrees=45
  swing2Motion=Free
  twistLimitDegrees=45
  twistMotion=Free
 }
 {
  bone1Name=dyn_pouch_lower_01_r_MSTR
  bone2Name=pelvis
  swing1LimitDegrees=45
  swing1Motion=Free
  swing2LimitDegrees=45
  swing2Motion=Free
  twistLimitDegrees=45
  twistMotion=Free
 }
 {bone1Name=dyn_pouch_upper_01_l, bone2Name=dyn_pouch_upper_01_l_MSTR, swing1LimitDegrees=15, swing2LimitDegrees=25, twistLimitDegrees=5}
 {bone1Name=dyn_pouch_upper_01_r, bone2Name=dyn_pouch_upper_01_r_MSTR, swing1LimitDegrees=15, swing2LimitDegrees=25, twistLimitDegrees=5}
 {bone1Name=dyn_pouch_lower_01_l, bone2Name=dyn_pouch_lower_01_l_MSTR, swing1LimitDegrees=15, swing2LimitDegrees=45, twistLimitDegrees=5}
 {bone1Name=dyn_pouch_lower_01_r, bone2Name=dyn_pouch_lower_01_r_MSTR, swing1LimitDegrees=15, swing2LimitDegrees=45, twistLimitDegrees=5}
]
