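# Physics asset for the collar mesh of the medium male Traveler player body, setting up jointed collar flaps around the neck.
# Box bodies on front, side and back neck bones joint to tiny MSTR anchor bodies with limited swing and twist, tighter swing at the back.
# Anchors and a head-and-neck capsule chain link to a pelvis capsule through free-rotation constraints.
# Pouch and chain anchors are also present, without child bodies.
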PhysicsAsset /BRCosmetics/Characters/Player/Male/Medium/Bodies/M_MED_Traveler/Meshes/M_MED_Traveler_collar_Physics
bodies {
 dyn_chain_base_01_l_MSTR [{shapeType=Box, extentX=0.01, extentY=0.01, extentZ=0.01}]
 dyn_neck_01_l [{shapeType=Box, extentX=3.1687, extentY=9.5078, extentZ=14.1733, center=<5.4139 0.0001 0>, rotation=<rot 0 -89.9991 0>}]
 dyn_neck_01_l_MSTR [{shapeType=Box, extentX=0.01, extentY=0.01, extentZ=0.01}]
 dyn_neck_01_r [{shapeType=Box, extentX=3.1687, extentY=9.5078, extentZ=14.1733, center=<-5.4139 -0.0001 0>, rotation=<rot 0 -89.9991 180>}]
 dyn_neck_01_r_MSTR [{shapeType=Box, extentX=0.01, extentY=0.01, extentZ=0.01}]
 dyn_neck_back_01_MSTR [{shapeType=Box, extentX=0.01, extentY=0.01, extentZ=0.01}]
 dyn_neck_back_01_l_MSTR [{shapeType=Box, extentX=0.01, extentY=0.01, extentZ=0.01}]
 dyn_neck_back_01_r_MSTR [{shapeType=Box, extentX=0.01, extentY=0.01, extentZ=0.01}]
 dyn_neck_back_02 [{shapeType=Box, extentX=2.1687, extentY=13.0078, extentZ=14.1733, center=<7.2174 0.0001 0>, rotation=<rot 0 -89.9991 0>}]
 dyn_neck_back_02_l [{shapeType=Box, extentX=2.1687, extentY=9.5078, extentZ=14.1733, center=<5.4139 0.0001 0>, rotation=<rot 0 -89.9991 0>}]
 dyn_neck_back_02_r [{shapeType=Box, extentX=2.1687, extentY=9.5078, extentZ=14.1733, center=<-5.4139 -0.0001 0>, rotation=<rot 0 -89.9991 180>}]
 dyn_neck_front_01 [{shapeType=Box, extentX=3.1687, extentY=8.5078, extentZ=9.1733, center=<4.5825 0 0>, rotation=<rot 0 -89.9991 0>}]
 dyn_neck_front_01_MSTR [{shapeType=Box, extentX=0.01, extentY=0.01, extentZ=0.01}]
 dyn_neck_front_01_l [{shapeType=Box, extentX=3.1687, extentY=8.5078, extentZ=9.1733, center=<4.5825 0 0>, rotation=<rot 0 -89.9991 0>}]
 dyn_neck_front_01_l_MSTR [{shapeType=Box, extentX=0.01, extentY=0.01, extentZ=0.01}]
 dyn_neck_front_01_r [{shapeType=Box, extentX=3.1687, extentY=8.5078, extentZ=9.1733, center=<-4.5825 0 0>, rotation=<rot 0 -89.9991 180>}]
 dyn_neck_front_01_r_MSTR [{shapeType=Box, extentX=0.01, extentY=0.01, extentZ=0.01}]
 dyn_pouch_lower_01_l_MSTR [{shapeType=Box, extentX=0.01, extentY=0.01, extentZ=0.01}]
 dyn_pouch_lower_01_r_MSTR [{shapeType=Box, extentX=0.01, extentY=0.01, extentZ=0.01}]
 dyn_pouch_upper_01_l_MSTR [{shapeType=Box, extentX=0.01, extentY=0.01, extentZ=0.01}]
 dyn_pouch_upper_01_r_MSTR [{shapeType=Box, extentX=0.01, extentY=0.01, extentZ=0.01}]
 head [
  {shapeType=Capsule, length=4, radius=11.25, center=<5.5936 0.9679 -0.0002>, rotation=<rot -90 0 -10.0001>}
  {shapeType=Capsule, length=7.25, radius=7.75, center=<3.077 6.4793 0.0001>, rotation=<rot -90 -0.0201 0.0203>}
 ]
 neck_01 [{shapeType=Capsule, length=7.75, radius=6.5, center=<3.8928 1.5458 0>, rotation=<rot -90 0 -20.0046>}]
 neck_02 [{shapeType=Capsule, length=6.75, radius=7.5, center=<4.8851 0.9554 -0.0006>, rotation=<rot -90 0 -15.011>}]
 pelvis [{shapeType=Capsule, length=15.7206, radius=11.8169, center=<-0.1307 -0.6735 -0.0227>, rotation=<rot -0.6941 -90 0.0793>}]
}
constraints [
 {
  bone1Name=dyn_neck_01_l_MSTR
  bone2Name=pelvis
  swing1LimitDegrees=45
  swing1Motion=Free
  swing2LimitDegrees=45
  swing2Motion=Free
  twistLimitDegrees=45
  twistMotion=Free
 }
 {
  bone1Name=dyn_neck_back_01_MSTR
  bone2Name=pelvis
  swing1LimitDegrees=45
  swing1Motion=Free
  swing2LimitDegrees=45
  swing2Motion=Free
  twistLimitDegrees=45
  twistMotion=Free
 }
 {
  bone1Name=dyn_neck_01_r_MSTR
  bone2Name=pelvis
  swing1LimitDegrees=45
  swing1Motion=Free
  swing2LimitDegrees=45
  swing2Motion=Free
  twistLimitDegrees=45
  twistMotion=Free
 }
 {
  bone1Name=dyn_neck_front_01_r_MSTR
  bone2Name=pelvis
  swing1LimitDegrees=45
  swing1Motion=Free
  swing2LimitDegrees=45
  swing2Motion=Free
  twistLimitDegrees=45
  twistMotion=Free
 }
 {
  bone1Name=dyn_neck_front_01_MSTR
  bone2Name=pelvis
  swing1LimitDegrees=45
  swing1Motion=Free
  swing2LimitDegrees=45
  swing2Motion=Free
  twistLimitDegrees=45
  twistMotion=Free
 }
 {
  bone1Name=dyn_neck_front_01_l_MSTR
  bone2Name=pelvis
  swing1LimitDegrees=45
  swing1Motion=Free
  swing2LimitDegrees=45
  swing2Motion=Free
  twistLimitDegrees=45
  twistMotion=Free
 }
 {
  bone1Name=dyn_pouch_upper_01_l_MSTR
  bone2Name=pelvis
  swing1LimitDegrees=45
  swing1Motion=Free
  swing2LimitDegrees=45
  swing2Motion=Free
  twistLimitDegrees=45
  twistMotion=Free
 }
 {
  bone1Name=dyn_pouch_upper_01_r_MSTR
  bone2Name=pelvis
  swing1LimitDegrees=45
  swing1Motion=Free
  swing2LimitDegrees=45
  swing2Motion=Free
  twistLimitDegrees=45
  twistMotion=Free
 }
 {
  bone1Name=dyn_chain_base_01_l_MSTR
  bone2Name=pelvis
  swing1LimitDegrees=45
  swing1Motion=Free
  swing2LimitDegrees=45
  swing2Motion=Free
  twistLimitDegrees=45
  twistMotion=Free
 }
 {
  bone1Name=dyn_pouch_lower_01_l_MSTR
  bone2Name=pelvis
  swing1LimitDegrees=45
  swing1Motion=Free
  swing2LimitDegrees=45
  swing2Motion=Free
  twistLimitDegrees=45
  twistMotion=Free
 }
 {
  bone1Name=dyn_pouch_lower_01_r_MSTR
  bone2Name=pelvis
  swing1LimitDegrees=45
  swing1Motion=Free
  swing2LimitDegrees=45
  swing2Motion=Free
  twistLimitDegrees=45
  twistMotion=Free
 }
 {bone1Name=dyn_neck_front_01, bone2Name=dyn_neck_front_01_MSTR, swing1LimitDegrees=45, swing2LimitDegrees=45, twistLimitDegrees=5}
 {bone1Name=dyn_neck_front_01_l, bone2Name=dyn_neck_front_01_l_MSTR, swing1LimitDegrees=45, swing2LimitDegrees=45, twistLimitDegrees=5}
 {bone1Name=dyn_neck_front_01_r, bone2Name=dyn_neck_front_01_r_MSTR, swing1LimitDegrees=45, swing2LimitDegrees=45, twistLimitDegrees=5}
 {
  bone1Name=head
  bone2Name=neck_02
  swing1LimitDegrees=45
  swing1Motion=Free
  swing2LimitDegrees=45
  swing2Motion=Free
  twistLimitDegrees=45
  twistMotion=Free
 }
 {
  bone1Name=neck_01
  bone2Name=pelvis
  swing1LimitDegrees=45
  swing1Motion=Free
  swing2LimitDegrees=45
  swing2Motion=Free
  twistLimitDegrees=45
  twistMotion=Free
 }
 {
  bone1Name=neck_02
  bone2Name=neck_01
  swing1LimitDegrees=45
  swing1Motion=Free
  swing2LimitDegrees=45
  swing2Motion=Free
  twistLimitDegrees=45
  twistMotion=Free
 }
 {bone1Name=dyn_neck_01_l, bone2Name=dyn_neck_01_l_MSTR, swing1LimitDegrees=45, swing2LimitDegrees=45, twistLimitDegrees=5}
 {bone1Name=dyn_neck_01_r, bone2Name=dyn_neck_01_r_MSTR, swing1LimitDegrees=45, swing2LimitDegrees=45, twistLimitDegrees=5}
 {bone1Name=dyn_neck_back_02_l, bone2Name=dyn_neck_back_01_l_MSTR, swing1LimitDegrees=25, swing2LimitDegrees=25, twistLimitDegrees=5}
 {bone1Name=dyn_neck_back_02, bone2Name=dyn_neck_back_01_MSTR, swing1LimitDegrees=25, swing2LimitDegrees=25, twistLimitDegrees=5}
 {
  bone1Name=dyn_neck_back_01_r_MSTR
  bone2Name=pelvis
  swing1LimitDegrees=45
  swing1Motion=Free
  swing2LimitDegrees=45
  swing2Motion=Free
  twistLimitDegrees=45
  twistMotion=Free
 }
 {
  bone1Name=dyn_neck_back_01_l_MSTR
  bone2Name=pelvis
  swing1LimitDegrees=45
  swing1Motion=Free
  swing2LimitDegrees=45
  swing2Motion=Free
  twistLimitDegrees=45
  twistMotion=Free
 }
 {bone1Name=dyn_neck_back_02_r, bone2Name=dyn_neck_back_01_r_MSTR, swing1LimitDegrees=25, swing2LimitDegrees=25, twistLimitDegrees=5}
]
