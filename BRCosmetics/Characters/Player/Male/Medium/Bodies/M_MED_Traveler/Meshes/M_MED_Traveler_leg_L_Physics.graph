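# Physics asset for the left-leg mesh of the male medium Traveler cosmetic body, defining simulation bodies for a dangling chain.
# Three box links hang in series from a chain base, which has tight angular limits against a tiny master anchor body.
# That anchor and four similar pouch anchors connect to a pelvis capsule with free swing and twist.
# A large, thin, unconstrained floor box supplies an additional collision body.
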PhysicsAsset /BRCosmetics/Characters/Player/Male/Medium/Bodies/M_MED_Traveler/Meshes/M_MED_Traveler_leg_L_Physics
bodies {
 dyn_chain_01_l [{shapeType=Box, extentX=1.6027, extentY=5.8568, extentZ=8.3826, center=<3.4469 0 0.0001>, rotation=<rot 0 90 -89.998>}]
 dyn_chain_02_l [{shapeType=Box, extentX=5.8567, extentY=1.6071, extentZ=8.3826, center=<3.4469 0 0.0001>, rotation=<rot 0 90.0001 -89.994>}]
 dyn_chain_03_l [{shapeType=Box, extentX=1.6027, extentY=5.8568, extentZ=3.8826, center=<3.4469 0 0.0001>, rotation=<rot 0 90 -89.998>}]
 dyn_chain_base_01_l [{shapeType=Box, extentX=13.0809, extentY=8.0672, extentZ=19.1951}]
 dyn_chain_base_01_l_MSTR [{shapeType=Box, extentX=0.01, extentY=0.01, extentZ=0.01}]
 dyn_pouch_lower_01_l_MSTR [{shapeType=Box, extentX=0.01, extentY=0.01, extentZ=0.01}]
 dyn_pouch_lower_01_r_MSTR [{shapeType=Box, extentX=0.01, extentY=0.01, extentZ=0.01}]
 dyn_pouch_upper_01_l_MSTR [{shapeType=Box, extentX=0.01, extentY=0.01, extentZ=0.01}]
 dyn_pouch_upper_01_r_MSTR [{shapeType=Box, extentX=0.01, extentY=0.01, extentZ=0.01}]
 dyn_root_floor_COL [{shapeType=Box, extentX=150.51, extentY=187.51, extentZ=3, center=<-4 0 0>}]
 pelvis [{shapeType=Capsule, length=15.7206, radius=11.8169, center=<-0.1307 -0.6735 -0.0227>, rotation=<rot -0.6941 -90 0.0793>}]
}
constraints [
 {
  bone1Name=dyn_pouch_upper_01_l_MSTR
  bone2Name=pelvis
  swing1LimitDegrees=45
  swing1Motion=Free
  swing2LimitDegrees=45
  swing2Motion=Free
  twistLimitDegrees=45
  twistMotion=Free
 }
 {
  bone1Name=dyn_pouch_upper_01_r_MSTR
  bone2Name=pelvis
  swing1LimitDegrees=45
  swing1Motion=Free
  swing2LimitDegrees=45
  swing2Motion=Free
  twistLimitDegrees=45
  twistMotion=Free
 }
 {
  bone1Name=dyn_chain_base_01_l_MSTR
  bone2Name=pelvis
  swing1LimitDegrees=45
  swing1Motion=Free
  swing2LimitDegrees=45
  swing2Motion=Free
  twistLimitDegrees=45
  twistMotion=Free
 }
 {
  bone1Name=dyn_pouch_lower_01_l_MSTR
  bone2Name=pelvis
  swing1LimitDegrees=45
  swing1Motion=Free
  swing2LimitDegrees=45
  swing2Motion=Free
  twistLimitDegrees=45
  twistMotion=Free
 }
 {
  bone1Name=dyn_pouch_lower_01_r_MSTR
  bone2Name=pelvis
  swing1LimitDegrees=45
  swing1Motion=Free
  swing2LimitDegrees=45
  swing2Motion=Free
  twistLimitDegrees=45
  twistMotion=Free
 }
 {bone1Name=dyn_chain_01_l, bone2Name=dyn_chain_base_01_l, swing1LimitDegrees=45, swing2LimitDegrees=45, twistLimitDegrees=5}
 {bone1Name=dyn_chain_02_l, bone2Name=dyn_chain_01_l, swing1LimitDegrees=45, swing2LimitDegrees=45, twistLimitDegrees=15}
 {bone1Name=dyn_chain_03_l, bone2Name=dyn_chain_02_l, swing1LimitDegrees=45, swing2LimitDegrees=45, twistLimitDegrees=15}
 {bone1Name=dyn_chain_base_01_l, bone2Name=dyn_chain_base_01_l_MSTR, swing1LimitDegrees=5, swing2LimitDegrees=5, twistLimitDegrees=5}
]
