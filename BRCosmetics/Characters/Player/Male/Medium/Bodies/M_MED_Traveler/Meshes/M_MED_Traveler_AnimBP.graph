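# Post-processes the M_MED_Traveler male body cosmetic, copying the attached parent mesh's pose and adding secondary motion to collar, neck and pouch bones.
# Head rotation relative to spine_05 and targeting state drive alphas for corrective neck-bone offsets, while pouches follow thighs and pelvis; the result is cached.
# That pose feeds three chained RigidBody simulations, swapped for head constraints and fixed rotations when rigid bodies are disabled or LOD exceeds 1.
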AnimBlueprint /BRCosmetics/Characters/Player/Male/Medium/Bodies/M_MED_Traveler/Meshes/M_MED_Traveler_AnimBP
target_skeleton=&/M_MED_Traveler_Skeleton
graphs [
 AnimGraph {nodes [
   n1 = RigidBody(ComponentPose=Local To Component(LocalPose=Use cached pose 'preSimCache'())) {
    bClampLinearTranslationLimitToRefPose=true
    bOverrideWorldGravity=true
    overridePhysicsAsset=&/M_MED_Traveler_Physics
    simulationSpace=BaseBoneSpace
    alphaBoolBlend {blendOption=Linear}
    baseBoneRef {boneName=spine_01}
    componentAppliedLinearAccClamp=<3000 3000 3000>
    componentLinearAccScale=<0.3 0.3 0.3>
    componentLinearVelScale=<0.3 0.3 0.3>
    overrideWorldGravity {z=-980}
   }
   n2 = RigidBody(ComponentPose=@n1) {
    bClampLinearTranslationLimitToRefPose=true
    bOverrideWorldGravity=true
    overridePhysicsAsset=&/M_MED_Traveler_leg_L_Physics
    simulationSpace=BaseBoneSpace
    alphaBoolBlend {blendOption=Linear}
    baseBoneRef {boneName=calf_l}
    componentAppliedLinearAccClamp=<3000 3000 3000>
    componentLinearAccScale=<0.3 0.3 0.3>
    componentLinearVelScale=<0.3 0.3 0.3>
    overrideWorldGravity {z=-980}
   }
   n3 = RigidBody(ComponentPose=@n2) {
    bClampLinearTranslationLimitToRefPose=true
    bOverrideWorldGravity=true
    overridePhysicsAsset=&/M_MED_Traveler_collar_Physics
    simulationSpace=BaseBoneSpace
    alphaBoolBlend {blendOption=Linear}
    baseBoneRef {boneName=spine_04}
    componentAppliedLinearAccClamp=<1500 1500 1500>
    componentLinearAccScale=<0.1 0.1 0.1>
    componentLinearVelScale=<0.1 0.1 0.1>
    overrideWorldGravity {z=-980}
   }
   n4 = "Transform (Modify) Bone - Bone: dyn_neck_back_02"(ComponentPose=@n3, Alpha=1, Translation="0,0,0", Rotation="0,15,0", Scale="1,1,1") {
    rotationMode=BMM_Additive
    rotationSpace=BCS_BoneSpace
    alphaBoolBlend {blendOption=Linear}
    boneToModify {boneName=dyn_neck_back_02}
    rotation {yaw=15}
   }
   n5 = "Transform (Modify) Bone - Bone: dyn_neck_back_03"(ComponentPose=@n4, Alpha=1, Translation="0,0,0", Rotation="0,45,0", Scale="1,1,1") {
    rotationMode=BMM_Additive
    rotationSpace=BCS_BoneSpace
    alphaBoolBlend {blendOption=Linear}
    boneToModify {boneName=dyn_neck_back_03}
    rotation {yaw=45}
   }
   n6 = "Transform (Modify) Bone - Bone: dyn_neck_back_01_l"(ComponentPose=@n5, Alpha=1, Translation="0,0,0", Rotation="0,15,0", Scale="1,1,1") {
    rotationMode=BMM_Additive
    rotationSpace=BCS_BoneSpace
    alphaBoolBlend {blendOption=Linear}
    boneToModify {boneName=dyn_neck_back_01_l}
    rotation {yaw=15}
   }
   n7 = "Transform (Modify) Bone - Bone: dyn_neck_back_01_r"(ComponentPose=@n6, Alpha=Get isTargetingAlpha() -> (isTargetingAlpha:double), Translation="0,0,0", Rotation="0,15,0", Scale="1,1,1") {
    rotationMode=BMM_Additive
    rotationSpace=BCS_BoneSpace
    alphaBoolBlend {blendOption=Linear}
    boneToModify {boneName=dyn_neck_back_01_r}
    rotation {yaw=15}
   }
   n8 = "Transform (Modify) Bone - Bone: dyn_neck_back_01_MSTR"(ComponentPose=@n7, Alpha=1, Translation="0,0,0", Rotation="0,-15,0", Scale="1,1,1") {
    rotationMode=BMM_Additive
    rotationSpace=BCS_BoneSpace
    alphaBoolBlend {blendOption=Linear}
    boneToModify {boneName=dyn_neck_back_01_MSTR}
    rotation {yaw=-15}
   }
   n9 = "Constraint - dyn_neck_back_01_MSTR (P)"(ComponentPose=Local To Component(LocalPose=Use cached pose 'preSimCache'()), Alpha=1, ConstraintWeights_0=1) {
    alphaBoolBlend {blendOption=Linear}
    boneToModify {boneName=dyn_neck_back_01_MSTR}
    constraintSetup [{targetBone {boneName=head}}]
    constraintWeights [1]
   }
   n10 = "Constraint - dyn_neck_back_01_r_MSTR (P)"(ComponentPose=@n9, Alpha=1, ConstraintWeights_0=1) {
    alphaBoolBlend {blendOption=Linear}
    boneToModify {boneName=dyn_neck_back_01_r_MSTR}
    constraintSetup [{targetBone {boneName=head}}]
    constraintWeights [1]
   }
   n11 = "Constraint - dyn_neck_back_01_l_MSTR (P)"(ComponentPose=@n10, Alpha=1, ConstraintWeights_0=1) {
    alphaBoolBlend {blendOption=Linear}
    boneToModify {boneName=dyn_neck_back_01_l_MSTR}
    constraintSetup [{targetBone {boneName=head}}]
    constraintWeights [1]
   }
   n12 = "Transform (Modify) Bone - Bone: dyn_neck_back_01_MSTR"(ComponentPose=@n11, Alpha=1, Translation="0,0,0", Rotation="0,-15,0", Scale="1,1,1") {
    rotationMode=BMM_Additive
    rotationSpace=BCS_BoneSpace
    alphaBoolBlend {blendOption=Linear}
    boneToModify {boneName=dyn_neck_back_01_MSTR}
    rotation {yaw=-15}
   }
   n13 = "Transform (Modify) Bone - Bone: dyn_neck_back_01_r_MSTR"(ComponentPose=@n12, Alpha=1, Translation="0,0,0", Rotation="0,-15,0", Scale="1,1,1") {
    rotationMode=BMM_Additive
    rotationSpace=BCS_BoneSpace
    alphaBoolBlend {blendOption=Linear}
    boneToModify {boneName=dyn_neck_back_01_r_MSTR}
    rotation {yaw=-15}
   }
   n14 = "Transform (Modify) Bone - Bone: dyn_neck_back_01_l_MSTR"(ComponentPose=@n13, Alpha=1, Translation="0,0,0", Rotation="0,-15,0", Scale="1,1,1") {
    rotationMode=BMM_Additive
    rotationSpace=BCS_BoneSpace
    alphaBoolBlend {blendOption=Linear}
    boneToModify {boneName=dyn_neck_back_01_l_MSTR}
    rotation {yaw=-15}
   }
   Output Pose(Result=Blend Poses by bool(
     BlendPose_0=Component To Local(ComponentPose=@n14)
     BlendPose_1=Component To Local(ComponentPose=@n8)
     bActiveValue=Get rigidsOnOffBool() -> (rigidsOnOffBool:bool)
    ) {blendType=Linear}) {layerGroup=None}
   n15 = Get headSwingAlpha() -> (headSwingAlpha:double)
   n16 = "Constraint - dyn_neck_back_01_MSTR (R)"(ComponentPose=Local To Component(LocalPose=Copy Pose From Mesh() {bUseAttachedParent=true}), Alpha=@n15, ConstraintWeights_0=1) {
    alphaBoolBlend {blendOption=Linear}
    boneToModify {boneName=dyn_neck_back_01_MSTR}
    constraintSetup [{transformType=Rotation, targetBone {boneName=head}}]
    constraintWeights [1]
   }
   n17 = "Constraint - dyn_neck_01_l_MSTR (R)"(ComponentPose=@n16, Alpha=@n15, ConstraintWeights_0=1) {
    alphaBoolBlend {blendOption=Linear}
    boneToModify {boneName=dyn_neck_01_l_MSTR}
    constraintSetup [{transformType=Rotation, targetBone {boneName=head}}]
    constraintWeights [1]
   }
   n18 = "Constraint - dyn_neck_01_r_MSTR (R)"(ComponentPose=@n17, Alpha=@n15, ConstraintWeights_0=1) {
    alphaBoolBlend {blendOption=Linear}
    boneToModify {boneName=dyn_neck_01_r_MSTR}
    constraintSetup [{transformType=Rotation, targetBone {boneName=head}}]
    constraintWeights [1]
   }
   n19 = "Constraint - dyn_neck_back_01_r_MSTR (R)"(ComponentPose=@n18, Alpha=@n15, ConstraintWeights_0=1) {
    alphaBoolBlend {blendOption=Linear}
    boneToModify {boneName=dyn_neck_back_01_r_MSTR}
    constraintSetup [{transformType=Rotation, targetBone {boneName=head}}]
    constraintWeights [1]
   }
   n20 = "Constraint - dyn_neck_back_01_l_MSTR (R)"(ComponentPose=@n19, Alpha=@n15, ConstraintWeights_0=1) {
    alphaBoolBlend {blendOption=Linear}
    boneToModify {boneName=dyn_neck_back_01_l_MSTR}
    constraintSetup [{transformType=Rotation, targetBone {boneName=head}}]
    constraintWeights [1]
   }
   n21 = "Transform (Modify) Bone - Bone: dyn_neck_back_01_MSTR"(ComponentPose=@n20, Alpha=1, Translation="0.000,-4,-0.000", Rotation="0,0,0", Scale="1,1,1") {
    rotationMode=BMM_Additive
    rotationSpace=BCS_BoneSpace
    translationMode=BMM_Additive
    translationSpace=BCS_BoneSpace
    alphaBoolBlend {blendOption=Linear}
    boneToModify {boneName=dyn_neck_back_01_MSTR}
    translation {y=-4}
   }
   n22 = "Transform (Modify) Bone - Bone: dyn_neck_01_r_MSTR"(ComponentPose=@n21, Alpha=1, Translation="0.000,2,-0.000", Rotation="0,0,0", Scale="1,1,1") {
    translationMode=BMM_Additive
    translationSpace=BCS_BoneSpace
    alphaBoolBlend {blendOption=Linear}
    boneToModify {boneName=dyn_neck_01_r_MSTR}
    translation {y=2}
   }
   n23 = "Transform (Modify) Bone - Bone: dyn_neck_01_l_MSTR"(ComponentPose=@n22, Alpha=1, Translation="0,-2,0", Rotation="0,0,0", Scale="1,1,1") {
    translationMode=BMM_Additive
    translationSpace=BCS_BoneSpace
    alphaBoolBlend {blendOption=Linear}
    boneToModify {boneName=dyn_neck_01_l_MSTR}
    translation {y=-2}
   }
   n24 = "Transform (Modify) Bone - Bone: dyn_neck_back_01_r_MSTR"(ComponentPose=@n23, Alpha=1, Translation="0.000,4,-0.000", Rotation="0,0,0", Scale="1,1,1") {
    rotationMode=BMM_Additive
    rotationSpace=BCS_BoneSpace
    translationMode=BMM_Additive
    translationSpace=BCS_BoneSpace
    alphaBoolBlend {blendOption=Linear}
    boneToModify {boneName=dyn_neck_back_01_r_MSTR}
    translation {y=4}
   }
   n25 = "Transform (Modify) Bone - Bone: dyn_neck_back_01_l_MSTR"(ComponentPose=@n24, Alpha=1, Translation="0,-4,0", Rotation="0,0,0", Scale="1,1,1") {
    rotationMode=BMM_Additive
    rotationSpace=BCS_BoneSpace
    translationMode=BMM_Additive
    translationSpace=BCS_BoneSpace
    alphaBoolBlend {blendOption=Linear}
    boneToModify {boneName=dyn_neck_back_01_l_MSTR}
    translation {y=-4}
   }
   n26 = Get headSwingAlpha_2() -> (headSwingAlpha_2:double)
   n27 = "Transform (Modify) Bone - Bone: dyn_neck_back_01_MSTR"(ComponentPose=@n25, Alpha=@n26, Translation="0,0,0", Rotation="0.000,-20,0.000", Scale="1,1,1") {
    rotationMode=BMM_Additive
    rotationSpace=BCS_BoneSpace
    translationMode=BMM_Additive
    translationSpace=BCS_BoneSpace
    alphaBoolBlend {blendOption=Linear}
    boneToModify {boneName=dyn_neck_back_01_MSTR}
    rotation {yaw=-20}
   }
   n28 = "Transform (Modify) Bone - Bone: dyn_neck_back_01_r_MSTR"(ComponentPose=@n27, Alpha=@n26, Translation="0,0,0", Rotation="0,-10,0", Scale="1,1,1") {
    rotationMode=BMM_Additive
    rotationSpace=BCS_BoneSpace
    translationMode=BMM_Additive
    translationSpace=BCS_BoneSpace
    alphaBoolBlend {blendOption=Linear}
    boneToModify {boneName=dyn_neck_back_01_r_MSTR}
    rotation {yaw=-10}
   }
   n29 = "Transform (Modify) Bone - Bone: dyn_neck_back_01_l_MSTR"(ComponentPose=@n28, Alpha=@n26, Translation="0,0,0", Rotation="0,-10,0", Scale="1,1,1") {
    rotationMode=BMM_Additive
    rotationSpace=BCS_BoneSpace
    translationMode=BMM_Additive
    translationSpace=BCS_BoneSpace
    alphaBoolBlend {blendOption=Linear}
    boneToModify {boneName=dyn_neck_back_01_l_MSTR}
    rotation {yaw=-10}
   }
   n30 = Get headRollAlphaL() -> (headRollAlphaL:double)
   n31 = "Transform (Modify) Bone - Bone: dyn_neck_back_01_l_MSTR"(ComponentPose=@n29, Alpha=@n30, Translation="0.000,-2,-0.000", Rotation="0,-35,0", Scale="1,1,1") {
    rotationMode=BMM_Additive
    rotationSpace=BCS_BoneSpace
    translationMode=BMM_Additive
    translationSpace=BCS_BoneSpace
    alphaBoolBlend {blendOption=Linear}
    boneToModify {boneName=dyn_neck_back_01_l_MSTR}
    rotation {yaw=-35}
    translation {y=-2}
   }
   n32 = "Transform (Modify) Bone - Bone: dyn_neck_01_l_MSTR"(ComponentPose=@n31, Alpha=@n30, Translation="0,-2,0", Rotation="0.000,-35,0.000", Scale="1,1,1") {
    rotationMode=BMM_Additive
    rotationSpace=BCS_BoneSpace
    translationMode=BMM_Additive
    translationSpace=BCS_BoneSpace
    alphaBoolBlend {blendOption=Linear}
    boneToModify {boneName=dyn_neck_01_l_MSTR}
    rotation {yaw=-35}
    translation {y=-2}
   }
   n33 = Get headRollAlphaR() -> (headRollAlphaR:double)
   n34 = "Transform (Modify) Bone - Bone: dyn_neck_back_01_r_MSTR"(ComponentPose=@n32, Alpha=@n33, Translation="0,-2,0", Rotation="0.000,-35,0.000", Scale="1,1,1") {
    rotationMode=BMM_Additive
    rotationSpace=BCS_BoneSpace
    translationMode=BMM_Additive
    translationSpace=BCS_BoneSpace
    alphaBoolBlend {blendOption=Linear}
    boneToModify {boneName=dyn_neck_back_01_r_MSTR}
    rotation {yaw=-35}
    translation {y=-2}
   }
   n35 = "Transform (Modify) Bone - Bone: dyn_neck_front_01_r_MSTR"(ComponentPose=@n34, Alpha=@n33, Translation="0,2.5,0", Rotation="0.000,-35,0.000", Scale="1,1,1") {
    rotationMode=BMM_Additive
    rotationSpace=BCS_BoneSpace
    translationMode=BMM_Additive
    translationSpace=BCS_BoneSpace
    alphaBoolBlend {blendOption=Linear}
    boneToModify {boneName=dyn_neck_front_01_r_MSTR}
    rotation {yaw=-35}
    translation {y=2.5}
   }
   n36 = "Transform (Modify) Bone - Bone: dyn_neck_01_r_MSTR"(ComponentPose=@n35, Alpha=@n33, Translation="0,-2,0", Rotation="0.000,-35,0.000", Scale="1,1,1") {
    rotationMode=BMM_Additive
    rotationSpace=BCS_BoneSpace
    translationMode=BMM_Additive
    translationSpace=BCS_BoneSpace
    alphaBoolBlend {blendOption=Linear}
    boneToModify {boneName=dyn_neck_01_r_MSTR}
    rotation {yaw=-35}
    translation {y=-2}
   }
   n37 = "Transform (Modify) Bone - Bone: dyn_neck_back_01_MSTR"(ComponentPose=@n36, Alpha=Get headSwingAlpha(), Translation="0.000,-3,0", Rotation="0,0,0", Scale="1,1,1") {
    rotationMode=BMM_Additive
    rotationSpace=BCS_BoneSpace
    translationMode=BMM_Additive
    translationSpace=BCS_BoneSpace
    alphaBoolBlend {blendOption=Linear}
    boneToModify {boneName=dyn_neck_back_01_MSTR}
    translation {y=-3}
   }
   n38 = "Constraint - dyn_pouch_lower_01_r_MSTR (PP)"!(ComponentPose=@n37, Alpha=1, ConstraintWeights_0=0.4, ConstraintWeights_1:float=0.6) -> (Pose) {
    alphaBoolBlend {blendOption=Linear}
    boneToModify {boneName=dyn_pouch_lower_01_r_MSTR}
    constraintSetup [{targetBone {boneName=thigh_twist_01_r}}, {targetBone {boneName=pelvis}}]
    constraintWeights [1,1]
   }
   n39 = "Constraint - dyn_pouch_lower_01_l_MSTR (PP)"!(ComponentPose=@n38, Alpha=1, ConstraintWeights_0=0.4, ConstraintWeights_1:float=0.6) -> (Pose) {
    alphaBoolBlend {blendOption=Linear}
    boneToModify {boneName=dyn_pouch_lower_01_l_MSTR}
    constraintSetup [{targetBone {boneName=thigh_twist_01_l}}, {targetBone {boneName=pelvis}}]
    constraintWeights [1,1]
   }
   Save cached pose 'preSimCache'(Pose=Component To Local(ComponentPose=@n39)) {cacheName=preSimCache}
  ]}
 EventGraph {nodes [
   Event BlueprintUpdateAnimation()
   > n1 = Sequence!(execute) -> (then_0, then_1, then_2:Exec)
   n2 = GetOwnerSkeletalMeshComponent()
   n3 = BreakRotator(InRot=BreakTransform(InTransform=GetDeltaTransformFromRefPose(self=@n2, BoneName=head, BaseName=spine_05)).Rotation)
   > Set headSwingAlpha(headSwingAlpha:double=MapRangeClamped(Value=@n3.Yaw, InRangeA=-10, InRangeB=0, OutRangeA=1, OutRangeB=0)) -> (Output_Get:double)
   > Set headSwingAlpha_2(headSwingAlpha_2:double=MapRangeClamped(Value=@n3.Yaw, InRangeA=10, InRangeB=0, OutRangeA=1, OutRangeB=0)) -> (Output_Get:double)
   n4 = BreakRotator(InRot=BreakTransform(InTransform=GetDeltaTransformFromRefPose(self=@n2, BoneName=head, BaseName=spine_05)).Rotation)
   > Set headRollAlphaL(headRollAlphaL:double=MapRangeClamped(Value=@n4, InRangeA=0, InRangeB=35)) -> (Output_Get:double)
   n5 = SelectFloat(A=1, B=0, bPickA=Get bIsTargeting())
   > Set headRollAlphaR(headRollAlphaR:double=MapRangeClamped(Value=@n4, InRangeA=0, InRangeB=-35, OutRangeA=@n5)) -> (Output_Get:double)
   > Set isTargetingAlpha(isTargetingAlpha:double=@n5) -> (Output_Get:double)
   n6 = Get CurrentLOD()
   n7 = Get bRigidBodyEnabled()
   Set rigidsOnOffBool(execute=@n1.then_1, rigidsOnOffBool:bool=OR Boolean(A=NOT Boolean(A=@n7), B="integer > integer"(A=@n6, B=1) {is=title K2Node_CallFunction})) -> (Output_Get:bool)
   TryGetPawnOwner()
  ]}
]
variables [
 headRollAlphaL {type=number}
 headRollAlphaR {type=number}
 headSwingAlpha {type=number}
 headSwingAlpha_2 {type=number}
 isTargetingAlpha {type=number}
 rigidsOnOffBool {type=boolean}
]
